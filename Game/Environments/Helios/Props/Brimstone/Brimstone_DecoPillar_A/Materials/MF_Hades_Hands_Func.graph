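# Material function for the Brimstone deco pillar that animates hands, apparently pressing and receding behind the surface, outputting MaskTexture and FingerPrints.
# Time, offset by vertex-color red, UV row and a random texture channel, drives per-hand sine cycles.
# These cycles shift hand UVs with height-refined view-direction parallax, add wobble and fade the mask by height.
# Fingerprints pulse from the same clock on undistorted UVs, and a static switch supports mirroring.
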MaterialFunction /Game/Environments/Helios/Props/Brimstone/Brimstone_DecoPillar_A/Materials/MF_Hades_Hands_Func
graph {nodes [
  n1 = FunctionInput() {bUsePreviewValueAsDefault=true, inputName=RandomTimeOffset, inputType=FunctionInput_Scalar, sortPriority=5, previewValue=<33.3 0 0 1>}
  n2 = Multiply(A=VertexColor().R, B=@n1)
  n3 = Time()
  n4 = ScalarParameter() {defaultValue=86.8054, parameterName=RandomUVoffset}
  n5 = VertexColor()
  n6 = AppendVector(A=ScalarParameter() {defaultValue=0.1184, parameterName=UV_x}, B=ScalarParameter() {defaultValue=0.5774, parameterName=UV_Y})
  n7 = StaticSwitch(
   True=Multiply(A=Constant2Vector() {g=1, r=-1}, B=@n6)
   False=@n6
   Value=FunctionInput(Preview=StaticBool()) {bUsePreviewValueAsDefault=true, inputName=IsMirrored, inputType=FunctionInput_StaticBool, sortPriority=10}
  )
  n8 = Add(
   A=AppendVector(A=Constant(), B=Multiply(A=@n4, B=@n5.R))
   B=MaterialFunctionCall(UVs=TextureCoordinate(), Texture Scale=@n7) {materialFunction=&ScaleUVsByCenter}.UVs
  )
  n9 = Add(A=@n8, B=FunctionInput() {bUsePreviewValueAsDefault=true, inputName=UVsOffset, inputType=FunctionInput_Vector2})
  n10 = Add(A=@n9, B=FunctionInput() {bUsePreviewValueAsDefault=true, inputName=UV_Distortion, inputType=FunctionInput_Vector2})
  n11 = Multiply(A=Ceil(None=ComponentMask(None=@n10) {g=true}), B=ScalarParameter() {defaultValue=43.0121, parameterName=RandomTimeOffset2})
  n12 = TextureSample(UVs=@n10) {samplerType=SAMPLERTYPE_LinearGrayscale, texture=/Game/Environments/Helios/Props/Brimstone/Brimstone_DecoPillar_A/Textures/Hades_Hands_M_2}
  n13 = ScalarParameter() {defaultValue=62.0875, parameterName=RandomTimeOffset}
  n14 = Add(A=Add(A=Add(A=@n2, B=@n3), B=@n11), B=Add(A=Multiply(A=@n12.B, B=@n13), B=@n1))
  n15 = ScalarParameter() {defaultValue=-0.092, parameterName=SpeedVariation}
  n16 = Multiply(A=@n14, B=Add(A=ScalarParameter() {defaultValue=0.18, parameterName=Speed}, B=Multiply(A=@n15, B=@n12.B)))
  n17 = Sine(None=@n16)
  n18 = Power(Base=Divide(A=Add(A=@n17)), Exp=ScalarParameter() {defaultValue=1.714, parameterName=ANimExponent})
  n19 = ScalarParameter() {defaultValue=-1.2284, parameterName=ANim_Pause_low}
  n20 = ScalarParameter() {defaultValue=0.8918, parameterName=ANim_Pause_high}
  n21 = Divide(A=Add(A=Cosine(None=@n16)))
  Add(A=@n18, B=Saturate(None=LinearInterpolate(A=@n19, B=@n20, Alpha=@n21)))
  Constant2Vector()
  Constant3Vector()
  n22 = MaterialFunctionCall(Multiplier=Constant() {r=1}) {materialFunction=&Pi}
  n23 = Frac(None=Divide(A=Ceil(None=@n16), B=@n22.Result))
  Divide(A=@n23, B=@n22.Result)
  DotProduct(A=CameraVectorWS(), B=VertexNormalWS())
  n24 = ScalarParameter() {defaultValue=0.3624, parameterName=Offset_FingerPrintTime_Fracoffset}
  n25 = OneMinus(None=Frac(None=Add(A=@n16, B=@n24)))
  n26 = Add(A=@n16, B=ScalarParameter() {defaultValue=0.4664, parameterName=Offset_FingerPrintTime_Fracoffset2})
  n27 = Saturate(None=Multiply(A=Frac(None=@n26), B=ScalarParameter() {defaultValue=2, parameterName=Offset_FingerPrintTime_Speed2}))
  FunctionOutput(None=Multiply(
    A=Multiply(A=@n25, B=@n27)
    B=TextureSample(UVs=@n9) {texture=/Game/Environments/Helios/Props/Brimstone/Brimstone_DecoPillar_A/Textures/Hades_Hands_Prints}.G
   )) {outputName=FingerPrints}
  n28 = CameraVectorWS()
  n29 = Transform(None=@n28) {transformSourceType=TRANSFORMSOURCE_World, transformType=TRANSFORM_Tangent}
  n30 = ComponentMask(None=@n29) {g=true, r=true}
  n31 = ScalarParameter() {defaultValue=0.01, parameterName=Offset}
  n32 = ScalarParameter() {defaultValue=-0.02, parameterName=Offset2}
  n33 = Saturate(None=@n18)
  n34 = DotProduct(A=@n28, B=VertexNormalWS())
  n35 = Divide(A=Divide(A=LinearInterpolate(A=@n31, B=@n32, Alpha=@n33), B=@n7), B=@n34)
  n36 = Constant()
  n37 = AppendVector(A=ScalarParameter() {defaultValue=-0.01, parameterName=Offset_X_1}, B=@n36)
  n38 = AppendVector(A=ScalarParameter() {defaultValue=0.05, parameterName=Offset_X_2}, B=@n36)
  n39 = Add(A=@n10, B=Add(A=Multiply(A=@n30, B=@n35), B=LinearInterpolate(A=@n37, B=@n38, Alpha=@n33)))
  n40 = OneMinus(None=TextureSample(UVs=@n39) {samplerType=SAMPLERTYPE_Masks, texture=/Game/Environments/Helios/Props/Brimstone/Brimstone_DecoPillar_A/Textures/Hades_Hands_M}.G)
  n41 = ScalarParameter() {defaultValue=-0.048, parameterName=Offset3}
  n42 = Divide(A=@n41, B=@n7)
  n43 = Multiply(A=@n30, B=Divide(A=Multiply(A=@n40, B=@n42), B=@n34))
  n44 = Frac(None=@n1)
  n45 = LinearInterpolate(
   A=ScalarParameter() {defaultValue=0.464, parameterName=SineUVSpeed}
   B=ScalarParameter() {defaultValue=0.216, parameterName=SineUVSpeed2}
   Alpha=Frac(None=Add(A=@n44, B=@n11))
  )
  n46 = Power(Base=Sine(None=Multiply(A=@n14, B=@n45)))
  n47 = ScalarParameter() {defaultValue=0.01, parameterName=SineUVOffset}
  n48 = TextureSample(UVs=Add(A=Add(A=@n43, B=@n39), B=Multiply(A=@n46, B=@n47))) {samplerType=SAMPLERTYPE_Masks, texture=&Hades_Hands_M}
  n49 = ScalarParameter() {defaultValue=-0.136, parameterName=FadeHeight_1}
  n50 = ScalarParameter() {defaultValue=0.112, parameterName=FadeHeight_2}
  n51 = Power(
   Base=Subtract(A=@n48.G, B=LinearInterpolate(A=@n49, B=@n50, Alpha=@n33))
   Exp=ScalarParameter() {defaultValue=4.7491, parameterName=HeightFalloff}
  )
  FunctionOutput(None=Multiply(A=@n48.R, B=Saturate(None=@n51))) {outputName=MaskTexture, sortPriority=10}
  LinearInterpolate()
  n52 = Sine(None=Add(A=@n16, B=ScalarParameter() {defaultValue=3.8016, parameterName=Offset_FingerPrintTime}) {constB=0.5})
  n53 = ScalarParameter() {defaultValue=0.2504, parameterName=Offset_FingerPrintTime_Bias}
  n54 = Saturate(None=Sign(None=Add(A=@n52, B=@n53)))
  Multiply(A=@n54, B=@n25)
  Multiply(A=@n41)
  Multiply(A=Frac())
  Multiply(B=@n29)
  OneMinus(None=@n54)
  OneMinus(None=@n21)
  Panner(Coordinate=@n8) {speedY=0.03}
  PixelNormalWS()
  Saturate(None=Multiply(A=@n23) {constB=2})
  Sign(None=Sine(None=Multiply(A=@n22.Result, B=@n23)))
  TextureCoordinate()
 ]}
