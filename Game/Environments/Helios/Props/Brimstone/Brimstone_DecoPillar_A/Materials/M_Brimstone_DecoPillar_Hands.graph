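# Emissive material for the Brimstone DecoPillar A prop: dark hand silhouettes with glowing fingerprints over a red or teal glow, chosen by static switch.
# Four offset MF_Hades_Hands_Func layers, two mirrored, are warped by rock and panning water normals and combined.
# A camera-facing virtual plane creates the radial glow, masked by a parallax-shifted, panning edge gradient.
# Emissive compensates for eye adaptation, brightens under ray tracing, and the attributes feed MF_FN_DestructibleDamage.
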Material /Game/Environments/Helios/Props/Brimstone/Brimstone_DecoPillar_A/Materials/M_Brimstone_DecoPillar_Hands
graph {
 properties {bUseMaterialAttributes=true, bUsedWithNanite=true, pixelDepthOffsetMode=PDOM_Legacy, usedShadingModels=Default Lit}
 nodes [
  n1 = CameraPositionWS()
  n2 = ObjectPositionWS()
  DotProduct(A=CameraVectorWS(), B=Normalize(VectorInput=Subtract(A=@n1, B=@n2)))
  n3 = VectorParameter() {parameterName=ColorRed, defaultValue=<rgba 0.0573 0.008 0.004 1>}
  n4 = VectorParameter() {parameterName=Color, defaultValue=<rgba 0.0059 0.0556 0.0573 1>}
  n5 = VectorParameter() {parameterName=Color2Red, defaultValue=<rgba 0.8281 0.2061 0.0735 1>}
  n6 = VectorParameter() {parameterName=Color2, defaultValue=<rgba 0.103 0.8281 0.437 1>}
  n7 = Normalize(VectorInput=Multiply(A=Constant3Vector() {constant=<rgba 1 1 0 0>}, B=CameraVectorWS()))
  n8 = Constant3Vector() {constant=<rgba 0 0 1 0>}
  n9 = ObjectPositionWS()
  n10 = Distance(
   A=MaterialFunctionCall(
    Plane Normal Axis=@n7
    Plane X Axis=CrossProduct(A=@n7, B=@n8)
    Plane Center=@n9
    UV Scale=ScalarParameter() {defaultValue=7614.1904, parameterName=VirutalPlaneScale}
   ) {materialFunction=&VirtualPlaneCoordinates}.UVs
   B=Constant2Vector() {g=0.5, r=0.5}
  )
  n11 = ScalarParameter() {defaultValue=1.034, parameterName=DistanceFalloff}
  n12 = Power(Base=OneMinus(None=Divide(A=@n10, B=@n11)), Exp=ScalarParameter() {defaultValue=107.0971, parameterName=Falloff})
  n13 = TextureSample(UVs=Multiply(A=TextureCoordinate(), B=ScalarParameter() {defaultValue=2, parameterName=NormalTiling})) {samplerType=SAMPLERTYPE_Normal, texture=/Game/Environments/Asteria/Terrain/Textures/T_Asteria_FrozenRockCliff_01_N}
  n14 = TextureCoordinate()
  n15 = ScalarParameter() {defaultValue=2.6428, parameterName=WaterNormalTiling}
  n16 = ScalarParameter() {defaultValue=0.09, parameterName=WaterPaning_Speed}
  n17 = TextureSample(UVs=Panner(Coordinate=Multiply(A=@n14, B=@n15), Speed=AppendVector(A=@n16, B=@n16)) {speedY=0.01}) {samplerType=SAMPLERTYPE_Normal, texture=/FortWater/Textures/T_Water_TilingNormal_Waves_01}
  n18 = Multiply(
   A=Add(A=ComponentMask(None=@n13.RGB) {g=true, r=true}, B=ComponentMask(None=@n17.RGB) {g=true, r=true})
   B=ScalarParameter() {defaultValue=0.01, parameterName=UvDistortion}
  )
  n19 = MaterialFunctionCall!(
   UV_Distortion=@n18
   UVsOffset=AppendVector(A=ScalarParameter() {defaultValue=0.208, parameterName=UV_Offset_A_X}, B=ScalarParameter() {parameterName=UV_Offset_A_Y})
   RandomTimeOffset
   IsMirrored
  ) {materialFunction=&/MF_Hades_Hands_Func}
  n20 = MaterialFunctionCall(
   UV_Distortion=@n18
   UVsOffset=AppendVector(
    A=ScalarParameter() {defaultValue=-0.016, parameterName=UV_Offset_B_X}
    B=ScalarParameter() {defaultValue=-0.272, parameterName=UV_Offset_B_Y}
   )
   RandomTimeOffset=ScalarParameter() {defaultValue=45.8936, parameterName=RandomB}
  ) {materialFunction=&MF_Hades_Hands_Func}
  n21 = Max(A=@n19.MaskTexture, B=@n20.MaskTexture)
  n22 = MaterialFunctionCall(
   UV_Distortion=@n18
   UVsOffset=AppendVector(A=ScalarParameter() {defaultValue=0.336, parameterName=UV_Offset_C_X}, B=ScalarParameter() {defaultValue=0.128, parameterName=UV_Offset_C_Y})
   RandomTimeOffset=ScalarParameter() {defaultValue=102.4349, parameterName=RandomC}
   IsMirrored=StaticBool() {value=true}
  ) {materialFunction=&MF_Hades_Hands_Func}
  n23 = MaterialFunctionCall(
   UV_Distortion=@n18
   UVsOffset=AppendVector(A=ScalarParameter() {defaultValue=0.048, parameterName=UV_Offset_D_X}, B=ScalarParameter() {defaultValue=0.368, parameterName=UV_Offset_D_Y})
   RandomTimeOffset=ScalarParameter() {defaultValue=39.9519, parameterName=RandomD}
   IsMirrored=StaticBool() {value=true}
  ) {materialFunction=&MF_Hades_Hands_Func}
  n24 = LinearInterpolate(
   A=LinearInterpolate(
    A=StaticSwitchParameter(True=@n3.RGB, False=@n4.RGB) {defaultValue=true, parameterName=IsRed}
    B=StaticSwitchParameter(True=@n5.RGB, False=@n6.RGB) {defaultValue=true, parameterName=IsRed}
    Alpha=@n12
   )
   B=Constant3Vector()
   Alpha=Max(A=Max(A=@n21, B=@n22.MaskTexture), B=@n23.MaskTexture)
  )
  n25 = Transform()
  n26 = Normalize(VectorInput=Subtract(A=WorldPosition().XYZ, B=@n9))
  LinearInterpolate(
   A=VectorParameter() {parameterName=Lightinh, defaultValue=<rgba 1 1 1 1>}.RGB
   B=@n24
   Alpha=Power(Base=DotProduct(A=@n25, B=@n26), Exp=ScalarParameter() {defaultValue=0.4872, parameterName=LightingFalloff})
  )
  n27 = CameraVectorWS()
  n28 = Transform(None=@n27) {transformSourceType=TRANSFORMSOURCE_World, transformType=TRANSFORM_Tangent}
  n29 = ScalarParameter() {defaultValue=-0.15, parameterName=OffsetGradientEdge}
  n30 = DotProduct(A=@n27, B=VertexNormalWS())
  n31 = ScalarParameter() {defaultValue=0.224, parameterName=WidthGradient}
  n32 = Constant() {r=1}
  n33 = Add(
   A=Multiply(A=ComponentMask(None=@n28) {g=true, r=true}, B=Divide(A=@n29, B=@n30))
   B=MaterialFunctionCall(UVs=TextureCoordinate(), Texture Scale=AppendVector(A=@n31, B=@n32)) {materialFunction=&ScaleUVsByCenter}.UVs
  )
  n34 = Power(
   Base=TextureSample(UVs=Panner(Coordinate=@n33) {speedY=0.1}) {samplerType=SAMPLERTYPE_Masks, texture=/Game/Environments/Helios/Props/Brimstone/Brimstone_DecoPillar_A/Textures/Hades_Hands_M}.B
   Exp=ScalarParameter() {defaultValue=3.8492, parameterName=GradientEdgeFalloff}
  )
  n35 = LinearInterpolate(
   A=StaticSwitchParameter(
    True=VectorParameter() {parameterName=FingerPrintColor2Red, defaultValue=<rgba 0.0521 0.0102 0 1>}.RGB
    False=VectorParameter() {parameterName=FingerPrintColor2, defaultValue=<rgba 0 0.0521 0.0366 1>}.RGB
   ) {defaultValue=true, parameterName=IsRed}
   B=StaticSwitchParameter(
    True=VectorParameter() {parameterName=FingerPrintColorRed, defaultValue=<rgba 0.577 0.1773 0 1>}.RGB
    False=VectorParameter() {parameterName=FingerPrintColor, defaultValue=<rgba 0 0.577 0.4051 1>}.RGB
   ) {defaultValue=true, parameterName=IsRed}
   Alpha=@n12
  )
  n36 = Add(A=@n19.FingerPrints, B=@n20.FingerPrints)
  n37 = Saturate(None=Add(A=Add(A=@n36, B=@n22.FingerPrints), B=@n23.FingerPrints))
  n38 = Divide(A=Add(A=Multiply(A=@n34, B=@n24), B=Multiply(A=@n35, B=@n37)), B=EyeAdaptation().EyeAdaptation)
  n39 = SetMaterialAttributes!(
   MaterialAttributes
   Base Color=Constant3Vector()
   Roughness=ScalarParameter() {parameterName=Roughness}
   Emissive Color=RayTracingQualitySwitch(Normal=@n38, RayTraced=Multiply(A=@n38) {constB=5})
   Normal=MaterialFunctionCall(Normal=@n13.RGB, Flatness=ScalarParameter() {defaultValue=0.8703, parameterName=Normal}) {materialFunction=&FlattenNormal}.Result
  ) {attributeSetTypes [
    #1
    #2
    #3
    #4
   ]}
  MaterialFunctionCall(Material_In=@n39) {materialFunction=&MF_FN_DestructibleDamage}
  Multiply()
 ]
}
